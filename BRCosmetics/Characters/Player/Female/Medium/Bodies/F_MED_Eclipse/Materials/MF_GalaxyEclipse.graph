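# Material function that overlays an animated galaxy effect on the Eclipse female character skin, behind a Use Galaxy Eclipse static switch.
# Two masked, tinted nebula layers use rotating triplanar camera-vector and reflection-vector projections, with FOV-adjusted tiling.
# Rotating and twinkling stars, Fresnel edge color, sequin highlights computed from the directional light, and an optional tights rim build the emissive.
# The result, with black base color, blends over the incoming material attributes.
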MaterialFunction /BRCosmetics/Characters/Player/Female/Medium/Bodies/F_MED_Eclipse/Materials/MF_GalaxyEclipse
graph {nodes [
  n1 = NamedRerouteUsage()
  n2 = FunctionInput() {inputType=FunctionInput_MaterialAttributes}
  n3 = GetMaterialAttributes(None=@n2) {attributeGetTypes [
    #1
    #2
    #3
    #4
    #5
   ]}
  n4 = FunctionInput() {inputName=GalaxyMaskA, inputType=FunctionInput_Scalar, previewValue=<0 0 1 0>}
  n5 = TextureObjectParameter() {
   group=Galaxy Eclipse
   parameterName=GalaxyTextureA
   texture=/Game/Characters/Player/Male/Medium/Bodies/M_Med_Soldier_04/Skins/Galaxy/Textures/T_FN_Nebula_Pink
  }
  n6 = ScalarParameter() {defaultValue=0.2, group=Galaxy Eclipse, parameterName=GalaxyATiling}
  n7 = MaterialFunctionCall() {materialFunction=/Engine/Functions/Engine_MaterialFunctions02/Utility/FOV}
  n8 = Constant() {r=0.414}
  n9 = Constant() {r=1.091}
  n10 = Constant()
  n11 = Constant() {r=1}
  n12 = LinearInterpolate(
   A=Constant() {r=3}
   B=Constant() {r=1}
   Alpha=Saturate(None=MaterialFunctionCall(Input=@n7."FOV ", Input Low=@n8, Input High=@n9, Target Low=@n10, Target High=@n11) {materialFunction=&RemapValueRange}.Result)
  )
  n13 = VectorParameter() {group=Galaxy Eclipse, parameterName=GalaxyOffset, defaultValue=<rgba 0 0 0 1>}
  n14 = Add(A=Multiply(A=NamedRerouteUsage(), B=ScalarParameter() {defaultValue=0.012, group=Galaxy Eclipse, parameterName=WarpIntensity}), B=@n13.RGB)
  n15 = Constant() {r=6}
  n16 = Time()
  n17 = Constant()
  n18 = ScalarParameter() {defaultValue=0.02, group=Galaxy Eclipse, parameterName=Galaxy Rotation Speed}
  n19 = ScalarParameter() {defaultValue=-0.02, group=Galaxy Eclipse, parameterName=Galaxy Rotation Speed_1}
  n20 = Multiply(
   A=LinearInterpolate(A=@n16, B=Multiply(A=@n16, B=@n17), Alpha=ScalarParameter() {group=Galaxy Eclipse, parameterName=Background Rotation})
   B=AppendVector(A=AppendVector(A=@n18, B=@n19), B=ScalarParameter() {defaultValue=0.1, group=Galaxy Eclipse, parameterName=Galaxy Rotation Speed_2})
  )
  n21 = ComponentMask(None=MaterialFunctionCall!(CameraVector, Rotation Speed=@n20, Texture=@n5, Tiling=Multiply(A=@n6, B=@n12), Offset=@n14, AxisFadeContrast=@n15) {materialFunction=/Game/Weapons/WeaponSkins/MaterialFunction/TriplanarCameraVector_Rotated}.Result) {b=true}
  n22 = ScalarParameter() {defaultValue=0.1, group=Galaxy Eclipse, parameterName=GalaxyA_MaskSoften, sliderMax=1}
  n23 = Subtract(B=@n22)
  n24 = Multiply(
   A=Add(A=Multiply(A=@n21, B=@n23), B=@n22)
   B=VectorParameter() {group=Galaxy Eclipse, parameterName=GalaxyA_Tint, defaultValue=<rgba 0.5 0 1 0>}.RGB
  )
  n25 = FunctionInput() {inputName=GalaxyMaskB, inputType=FunctionInput_Scalar, previewValue=<0 0 1 0>}
  n26 = ReflectionVectorWS(CustomWorldNormal=MaterialFunctionCall(Normal=NamedRerouteUsage(), Flatness=ScalarParameter() {group=Galaxy Eclipse, parameterName=NormalFlatnessReflectionVector, sliderMax=1}) {materialFunction=&FlattenNormal}.Result) {bNormalizeCustomWorldNormal=true}
  n27 = Multiply(A=@n20) {constB=-1}
  n28 = Multiply(A=ScalarParameter() {defaultValue=0.2, group=Galaxy Eclipse, parameterName=GalaxyBTiling}, B=@n12)
  n29 = VectorParameter() {group=Galaxy Eclipse, parameterName=GalaxyB_Offset, defaultValue=<rgba 0.5 0.5 0.5 1>}
  n30 = Add(A=@n29.RGB, B=@n14)
  n31 = ScalarParameter() {defaultValue=0.1, group=Galaxy Eclipse, parameterName=GalaxyB_MaskSoften, sliderMax=1}
  n32 = Multiply(
   A=ComponentMask(None=MaterialFunctionCall(CameraVector=@n26, Rotation Speed=@n27, Texture=@n5, Tiling=@n28, Offset=@n30, AxisFadeContrast=@n15) {materialFunction=&TriplanarCameraVector_Rotated}.Result) {b=true}
   B=Subtract(B=@n31)
  )
  n33 = Multiply(
   A=@n25
   B=Multiply(A=Add(A=@n32, B=@n31), B=VectorParameter() {group=Galaxy Eclipse, parameterName=GalaxyB_Tint, defaultValue=<rgba 0 0.1 1 0>}.RGB)
  )
  n34 = StaticSwitchParameter(True=Add(A=Multiply(A=@n4, B=@n24), B=@n33), False=@n33) {group=Galaxy Eclipse, parameterName=UseGalaxyA}
  n35 = Desaturation(None=@n34, Fraction=ScalarParameter() {defaultValue=0.5, group=Galaxy Eclipse, parameterName=Star Saturation})
  n36 = NamedRerouteUsage()
  n37 = LinearInterpolate(Alpha=@n36) {constA=-1}
  n38 = Multiply(A=Time(), B=ScalarParameter() {defaultValue=-0.125, group=Galaxy Eclipse, parameterName=Small Star Rotation speed})
  n39 = TextureObjectParameter() {
   group=Galaxy Eclipse
   parameterName=Stars
   samplerType=SAMPLERTYPE_Masks
   texture=/Game/Characters/Player/Male/Medium/Bodies/M_Med_Soldier_04/Skins/Galaxy/Textures/Celestial_T_Stars
  }
  n40 = ScalarParameter() {defaultValue=5, group=Galaxy Eclipse, parameterName=Small Star Tiling}
  n41 = Add(A=Multiply(A=@n29.RGB, B=@n36), B=@n13.RGB)
  n42 = ComponentMask(None=MaterialFunctionCall(Rotation Speed=Multiply(A=@n37, B=@n38), Texture=@n39, Tiling=Multiply(A=@n40, B=@n12), Offset=@n41) {materialFunction=&TriplanarCameraVector_Rotated}.Result) {b=true}
  n43 = LinearInterpolate(
   A=@n34
   B=Multiply(A=@n35, B=VectorParameter() {group=Galaxy Eclipse, parameterName=Star Brightness, defaultValue=<rgba 300 300 300 1>}.RGB)
   Alpha=Multiply(A=@n42, B=@n42)
  )
  n44 = Multiply(A=@n35, B=VectorParameter() {group=Galaxy Eclipse, parameterName=Star Brightness_1, defaultValue=<rgba 950 950 950 1>}.RGB)
  n45 = ScalarParameter() {defaultValue=2, group=Galaxy Eclipse, parameterName=Bright Star Tiling}
  n46 = ComponentMask(None=MaterialFunctionCall!(CameraVector, Texture=@n39, Tiling=Multiply(A=@n45, B=@n12), Offset=@n41, AxisFadeContrast) {materialFunction=/Engine/Functions/Engine_MaterialFunctions01/Texturing/TriplanarCameraVector}.Result) {r=true}
  n47 = TextureSample(UVs=Panner(Coordinate=TextureCoordinate() {uTiling=2, vTiling=2}) {speedX=0.133, speedY=0.2}, Tex=@n39) {
   samplerSource=SSM_Wrap_WorldGroupSettings
   samplerType=SAMPLERTYPE_Masks
   texture=&Celestial_T_Stars
  }
  n48 = Multiply(A=Multiply(A=@n46, B=@n46), B=@n47.G)
  n49 = NamedRerouteUsage()
  n50 = Add(
   A=Fresnel(ExponentIn=ScalarParameter() {defaultValue=2, group=Galaxy Eclipse, parameterName=GalaxyA_FresEX}) {baseReflectFraction=0}
   B=ScalarParameter() {group=Galaxy Eclipse, parameterName=GalaxyA_FresOffset, sliderMax=1, sliderMin=-1}
  ) {constB=-0.1}
  n51 = ScalarParameter() {defaultValue=2, group=Galaxy Eclipse, parameterName=GalaxyB_FresEX}
  n52 = Multiply(
   A=NamedRerouteUsage()
   B=Add(
    A=Fresnel(ExponentIn=@n51) {baseReflectFraction=0}
    B=ScalarParameter() {group=Galaxy Eclipse, parameterName=GalaxyB_FresOffset, sliderMax=1, sliderMin=-1}
   ) {constB=-0.1}
  )
  n53 = StaticSwitchParameter(True=Add(A=Multiply(A=@n49, B=@n50), B=@n52), False=@n52) {group=Galaxy Eclipse, parameterName=UseGalaxyA}
  n54 = Multiply(
   A=ScalarParameter() {defaultValue=1, group=Galaxy Eclipse, parameterName=Galaxy_GlobalBrightness}
   B=LinearInterpolate(
    A=LinearInterpolate(A=@n43, B=@n44, Alpha=@n48)
    B=VectorParameter() {group=Galaxy Eclipse, parameterName=EdgeFresnel, defaultValue=<rgba 0 0 0 1>}.RGB
    Alpha=Saturate(None=@n53)
   ) {constB=0}
  )
  n55 = Custom() {code="return View.DirectionalLightColor;", description=DirectionalLightColor, inputs []}
  n56 = VectorParameter() {group=Galaxy Eclipse, parameterName=SequinsColor, defaultValue=<rgba 0.7531 0.5417 1 1>}
  n57 = NamedRerouteUsage()
  n58 = TextureCoordinate()
  n59 = ScalarParameter() {defaultValue=20, group=Galaxy Eclipse, parameterName=SequinsTiling}
  n60 = Transform(None=TextureSample(UVs=Multiply(A=@n58, B=@n59)) {samplerType=SAMPLERTYPE_Normal, texture=/Game/Global/Textures/Default/Tileable_Patterns/T_SequinTile_N}.RGB)
  n61 = Constant3Vector() {constant=<rgba 0 0 -1 0>}
  n62 = Normalize(VectorInput=Add(
    A=Custom() {code="return View.DirectionalLightDirection;", description=DirectionalLightVector, inputs []}
    B=Transform(None=@n61) {transformSourceType=TRANSFORMSOURCE_Camera}
   ))
  n63 = Saturate(None=Power(Base=DotProduct(A=@n60, B=@n62), Exp=ScalarParameter() {defaultValue=200, group=Galaxy Eclipse, parameterName=HighlightExp, sliderMax=500}) {constExponent=200})
  n64 = Add(
   A=StaticSwitch(
    True=Add(A=@n3.EmissiveColor, B=@n54)
    False=@n54
    Value=StaticBoolParameter() {group=Galaxy Eclipse, parameterName=UseAdditiveEmissiveGalaxy}
   )
   B=Multiply(A=Multiply(A=@n55, B=@n56.RGB), B=Multiply(A=@n57, B=@n63))
  )
  n65 = VectorParameter() {group=Galaxy Eclipse, parameterName=Tights_Fresnel, defaultValue=<rgba 0.5 0 1 0>}
  n66 = ScalarParameter() {defaultValue=4, group=Galaxy Eclipse, parameterName=Tights_EdgeExp}
  n67 = Multiply(A=NamedRerouteUsage(), B=Multiply(A=Multiply(A=@n65.RGB, B=@n65.RGB), B=Fresnel(ExponentIn=@n66) {baseReflectFraction=0}))
  n68 = SetMaterialAttributes!(
   MaterialAttributes
   BaseColor=Constant()
   Specular=ScalarParameter() {group=Galaxy Eclipse, parameterName=Galaxy Specular}
   Roughness=Multiply(A=@n3.Roughness, B=ScalarParameter() {defaultValue=1, group=Galaxy Eclipse, parameterName=Galaxy Roughness})
   EmissiveColor=StaticSwitchParameter(True=Add(A=@n64, B=@n67), False=@n64) {group=Galaxy Eclipse, parameterName=UseGalaxyA}
   Normal=NamedRerouteDeclaration(None=@n3.Normal) {name=Normal, nodeColor=<rgba 0 0.4141 1 1>}
  ) {attributeSetTypes [
    #4
    #5
    #2
    #3
    #1
   ]}
  FunctionOutput(None=StaticSwitchParameter(True=BlendMaterialAttributes(A=@n2, B=@n68, Alpha=@n1) {vertexAttributeBlendType=UseA}, False=@n2) {group=Skin FX, parameterName=Use Galaxy Eclipse})
  NamedRerouteDeclaration(None=Add(
    A=NamedRerouteDeclaration(None=@n4) {name=GalaxyA_Mask, nodeColor=<rgba 0 0.4066 1 1>}
    B=NamedRerouteDeclaration(None=@n25) {name=GalaxyB_Mask, nodeColor=<rgba 0.5991 0 1 1>}
   )) {name=CombinedGalaxyMasks, nodeColor=<rgba 0.1953 1 0 1>}
 ]}
