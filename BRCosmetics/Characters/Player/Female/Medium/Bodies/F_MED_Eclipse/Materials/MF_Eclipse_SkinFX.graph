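# Material function that adds skin effects to the material attributes of the F_MED_Eclipse female character body.
# It samples the SkinFX_Mask texture and passes its RGBA with the incoming attributes to MF_MetalsLUT, which the UseMetalLUT static switch can bypass.
# The switched result then runs through MF_GalaxyEclipse, with the mask's blue channel driving both galaxy mask inputs, and becomes the output.
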MaterialFunction /BRCosmetics/Characters/Player/Female/Medium/Bodies/F_MED_Eclipse/Materials/MF_Eclipse_SkinFX
graph {nodes [
  n1 = TextureObjectParameter() {
   parameterName=SkinFX_Mask
   samplerType=SAMPLERTYPE_Masks
   texture=/BRCosmetics/Characters/Player/Female/Medium/Bodies/F_MED_Eclipse/Textures/T_F_MED_Eclipse_Body_FX
  }
  n2 = TextureSample(Tex=@n1) {samplerType=SAMPLERTYPE_Masks, texture=&T_F_MED_Eclipse_Body_FX}
  n3 = FunctionInput() {inputName=MaterialAttributesIn, inputType=FunctionInput_MaterialAttributes}
  n4 = MaterialFunctionCall(Mask=@n2.RGBA, Materials=@n3) {materialFunction=&MF_MetalsLUT}
  FunctionOutput(None=MaterialFunctionCall!(GalaxyMaskA=@n2.B, GalaxyMaskB=@n2.B, In=StaticSwitchParameter(True=@n4.Result, False=@n3) {group=Skin FX, parameterName=UseMetalLUT}) {materialFunction=&/MF_GalaxyEclipse}.Result)
 ]}
